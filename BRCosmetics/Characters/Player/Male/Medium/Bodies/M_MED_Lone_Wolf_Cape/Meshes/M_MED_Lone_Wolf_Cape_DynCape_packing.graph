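# Physics asset for the dynamic cape on the Lone Wolf male medium player body, with capsule bodies on cape bones, torso, arms and legs.
# Asymmetric back, side and front cape chains hang from small MSTR bodies constrained to the clavicles and upper spine.
# Chain joints have tight twist and mostly widening swing down the back, while MSTR and most body joints rotate freely.
# Free-rotation cross-constraints tie the lower left and middle chains together.
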PhysicsAsset /BRCosmetics/Characters/Player/Male/Medium/Bodies/M_MED_Lone_Wolf_Cape/Meshes/M_MED_Lone_Wolf_Cape_DynCape_packing
bodies {
 calf_l [{shapeType=Capsule, length=35.3125, radius=8.5363, center=<-31.4163 -2.6359 -0.0003>, rotation=<rot 90 0 -1.5716>}]
 calf_r [{shapeType=Capsule, length=35.3125, radius=8.5363, center=<31.4163 2.6359 0.0003>, rotation=<rot 90 0 -1.5716>}]
 cape_01_l [{shapeType=Capsule, length=13, radius=6, center=<5.9999 2 6>, rotation=<rot 90 0 0>}]
 cape_01_r [{shapeType=Capsule, length=4.5, radius=6.5, center=<5 -0.0001 6>, rotation=<rot 90 0 0>}]
 cape_02_l [{shapeType=Capsule, length=10.5, radius=7, center=<5 -0.5 3>, rotation=<rot 87.2076 -0.0002 -0.0002>}]
 cape_02_r [{shapeType=Capsule, length=5.25, radius=7.25, center=<5 0 5>, rotation=<rot 90 0 0>}]
 cape_03_l [{shapeType=Capsule, length=11.5, radius=8.25, center=<5 -0.3 2>, rotation=<rot 90 0 0>}]
 cape_03_r [{shapeType=Capsule, length=6.75, radius=7.25, center=<8 0 5>, rotation=<rot 90 0 0>}]
 cape_04_l [{shapeType=Capsule, length=6.5, radius=9.5, center=<5 0 2>, rotation=<rot 0 0 90>}]
 cape_05_l [{shapeType=Capsule, length=8.25, radius=10.25, center=<5 0 2>, rotation=<rot 0 0 90>}]
 cape_06_l [{shapeType=Capsule, length=5.5, radius=10.5, center=<5 0 2>, rotation=<rot 0 0 90>}]
 cape_front_01_l [{shapeType=Capsule, length=3.75, radius=5.75, center=<3 0 5>, rotation=<rot 0 0 90>}]
 cape_front_01_r [{shapeType=Capsule, length=3.75, radius=5.75, center=<5 0 5>, rotation=<rot 0 0 90>}]
 cape_front_02_l [{shapeType=Capsule, length=3.75, radius=5.75, center=<3 0 5>, rotation=<rot 0 0 90>}]
 cape_front_03_l [{shapeType=Capsule, length=3.75, radius=5.75, center=<3 0 5>, rotation=<rot 0 0 90>}]
 cape_front_l_MSTR [{shapeType=Capsule, length=0.505, radius=0.505, shapeName=cape_mid_r_MSTR_capsule, rotation=<rot 0 -90 0>}]
 cape_front_mid_01 [{shapeType=Capsule, length=3.75, radius=5.75, center=<3 0 4>, rotation=<rot 0 0 90>}]
 cape_front_mid_02 [{shapeType=Capsule, length=3.5, radius=6, center=<3 0 4>, rotation=<rot 0 0 90>}]
 cape_front_mid_MSTR [{shapeType=Capsule, length=0.505, radius=0.505, shapeName=cape_mid_r_MSTR_capsule, rotation=<rot 0 -90 0>}]
 cape_front_r_MSTR [{shapeType=Capsule, length=0.505, radius=0.505, shapeName=cape_mid_r_MSTR_capsule, rotation=<rot 0 -90 0>}]
 cape_l_MSTR [{shapeType=Capsule, length=0.505, radius=0.505, shapeName=cape_l_MSTR_capsule, rotation=<rot 0 -90 0>}]
 cape_mid_01_l [{shapeType=Capsule, length=15.5, radius=7.25, center=<5 0 5>, rotation=<rot 0 0 90>}]
 cape_mid_01_r [{shapeType=Capsule, length=17.75, radius=7.25, center=<5 -4 5>, rotation=<rot 0 0 90>}]
 cape_mid_02_l [{shapeType=Capsule, length=12, radius=7.75, center=<5 0 5>, rotation=<rot 0 0 98.7038>}]
 cape_mid_02_r [{shapeType=Capsule, length=17.25, radius=7.75, center=<5 -4 5>, rotation=<rot 0 0 90>}]
 cape_mid_03_l [{shapeType=Capsule, length=12.25, radius=7.5, center=<5 0 5>, rotation=<rot 0 0 98.7038>}]
 cape_mid_03_r [{shapeType=Capsule, length=17.5, radius=7.5, center=<5 -3 5>, rotation=<rot 0 0 90>}]
 cape_mid_04_l [{shapeType=Capsule, length=12.5, radius=7.25, center=<5 0 5>, rotation=<rot 0 0 98.7038>}]
 cape_mid_04_r [{shapeType=Capsule, length=17.75, radius=7.25, center=<5 -3 5>, rotation=<rot 0 0 90>}]
 cape_mid_l_MSTR [{shapeType=Capsule, length=0.505, radius=0.505, shapeName=cape_mid_l_MSTR_capsule, rotation=<rot 0 -90 0>}]
 cape_mid_r_MSTR [{shapeType=Capsule, length=0.505, radius=0.505, shapeName=cape_mid_r_MSTR_capsule, rotation=<rot 0 -90 0>}]
 cape_r_MSTR [{shapeType=Capsule, length=0.505, radius=0.505, shapeName=cape_r_MSTR_capsule, rotation=<rot 0 -90 0>}]
 clavicle_l [{shapeType=Capsule, length=20.1779, radius=8.5408, center=<12.7811 -0.8725 -2.3648>, rotation=<rot 87.2054 8.3691 -179.7298>}]
 clavicle_r [{shapeType=Capsule, length=14.1779, radius=8.5408, center=<-8.229 1.5203 2.1397>, rotation=<rot 87.206 8.3696 1.7391>}]
 lowerarm_l [{shapeType=Capsule, length=32.6377, radius=6.63, center=<18.227 0.9178 -2.1609>, rotation=<rot 82.7121 -0.0002 3.2014>}]
 lowerarm_r [{shapeType=Capsule, length=31.6377, radius=7.63, center=<-16.7494 -1.001 1.972>, rotation=<rot 82.7121 -0.0002 -176.7986>}]
 pelvis [{shapeType=Capsule, length=12.4375, radius=14.0363, center=<-3.1042 -0.6517 -0.0003>}]
 spine_01 [{shapeType=Capsule, length=10.625, radius=12.4738, center=<2.0183 -0.4464 0>}]
 spine_03 [{shapeType=Capsule, length=12.375, radius=12.75, center=<1.0161 0.8341 -0.0003>}]
 spine_04 [{shapeType=Capsule, length=12.125, radius=13, center=<-0.3332 1.3619 -0.0001>}]
 spine_05 [{shapeType=Capsule, length=14.875, radius=13, center=<-0.985 1.5707 0.0001>}]
 thigh_l [{shapeType=Capsule, length=25.875, radius=11.0988, center=<-22.9378 -0.6267 1.6968>, rotation=<rot 90 0 0>}]
 thigh_r [{shapeType=Capsule, length=25.875, radius=11.0988, center=<22.9378 0.6267 -1.6968>, rotation=<rot 90 0 180>}]
 upperarm_l [{shapeType=Capsule, length=13.4279, radius=8.7908, center=<24.1153 0.9363 -0.6192>, rotation=<rot 89.2881 150.8478 -33.6682>}]
 upperarm_r [{shapeType=Capsule, length=23.1779, radius=6.7908, center=<-13.2303 0.4345 0.7286>, rotation=<rot 89.2931 150.8495 146.3315>}]
}
constraints [
 {bone1Name=cape_02_l, bone2Name=cape_01_l, swing1LimitDegrees=15, swing2LimitDegrees=5, twistLimitDegrees=2}
 {bone1Name=cape_03_l, bone2Name=cape_02_l, swing1LimitDegrees=15, swing2LimitDegrees=10, twistLimitDegrees=2}
 {bone1Name=cape_04_l, bone2Name=cape_03_l, swing1LimitDegrees=15, swing2LimitDegrees=30, twistLimitDegrees=2}
 {bone1Name=cape_05_l, bone2Name=cape_04_l, swing1LimitDegrees=30, swing2LimitDegrees=30, twistLimitDegrees=2}
 {bone1Name=cape_06_l, bone2Name=cape_05_l, swing1LimitDegrees=30, swing2LimitDegrees=30, twistLimitDegrees=2}
 {bone1Name=cape_01_l, bone2Name=cape_l_MSTR, swing1LimitDegrees=15, swing2LimitDegrees=5, twistLimitDegrees=2}
 {bone1Name=cape_01_r, bone2Name=cape_r_MSTR, swing1LimitDegrees=15, swing2LimitDegrees=2, twistLimitDegrees=2}
 {bone1Name=cape_02_r, bone2Name=cape_01_r, swing1LimitDegrees=10, swing2LimitDegrees=15, twistLimitDegrees=2}
 {bone1Name=cape_03_r, bone2Name=cape_02_r, swing1LimitDegrees=10, swing2LimitDegrees=35, twistLimitDegrees=2}
 {bone1Name=cape_mid_01_l, bone2Name=cape_mid_l_MSTR, swing1LimitDegrees=15, swing2LimitDegrees=2, twistLimitDegrees=2}
 {bone1Name=cape_mid_02_l, bone2Name=cape_mid_01_l, swing1LimitDegrees=10, swing2LimitDegrees=10, twistLimitDegrees=2}
 {bone1Name=cape_mid_03_l, bone2Name=cape_mid_02_l, swing1LimitDegrees=10, swing2LimitDegrees=35, twistLimitDegrees=2}
 {bone1Name=cape_mid_04_l, bone2Name=cape_mid_03_l, swing1LimitDegrees=10, swing2LimitDegrees=35, twistLimitDegrees=2}
 {bone1Name=cape_mid_01_r, bone2Name=cape_mid_r_MSTR, swing1LimitDegrees=15, swing2LimitDegrees=2, twistLimitDegrees=2}
 {bone1Name=cape_mid_02_r, bone2Name=cape_mid_01_r, swing1LimitDegrees=10, swing2LimitDegrees=10, twistLimitDegrees=2}
 {bone1Name=cape_mid_03_r, bone2Name=cape_mid_02_r, swing1LimitDegrees=10, swing2LimitDegrees=35, twistLimitDegrees=2}
 {bone1Name=cape_mid_04_r, bone2Name=cape_mid_03_r, swing1LimitDegrees=10, swing2LimitDegrees=35, twistLimitDegrees=2}
 {
  bone1Name=spine_05
  bone2Name=spine_04
  swing1LimitDegrees=45
  swing1Motion=Free
  swing2LimitDegrees=45
  swing2Motion=Free
  twistLimitDegrees=45
  twistMotion=Free
 }
 {
  bone1Name=spine_04
  bone2Name=spine_03
  swing1LimitDegrees=45
  swing1Motion=Free
  swing2LimitDegrees=45
  swing2Motion=Free
  twistLimitDegrees=45
  twistMotion=Free
 }
 {
  bone1Name=spine_03
  bone2Name=spine_01
  swing1LimitDegrees=45
  swing1Motion=Free
  swing2LimitDegrees=45
  swing2Motion=Free
  twistLimitDegrees=45
  twistMotion=Free
 }
 {
  bone1Name=spine_01
  bone2Name=pelvis
  swing1LimitDegrees=45
  swing1Motion=Free
  swing2LimitDegrees=45
  swing2Motion=Free
  twistLimitDegrees=45
  twistMotion=Free
 }
 {
  bone1Name=lowerarm_l
  bone2Name=upperarm_l
  swing1LimitDegrees=45
  swing1Motion=Free
  swing2LimitDegrees=45
  swing2Motion=Free
  twistLimitDegrees=45
  twistMotion=Free
 }
 {
  bone1Name=thigh_l
  bone2Name=pelvis
  swing1LimitDegrees=45
  swing1Motion=Free
  swing2LimitDegrees=45
  swing2Motion=Free
  twistLimitDegrees=45
  twistMotion=Free
 }
 {
  bone1Name=calf_l
  bone2Name=thigh_l
  swing1LimitDegrees=45
  swing1Motion=Free
  swing2LimitDegrees=45
  swing2Motion=Free
  twistLimitDegrees=45
  twistMotion=Free
 }
 {
  bone1Name=calf_r
  bone2Name=thigh_r
  swing1LimitDegrees=45
  swing1Motion=Free
  swing2LimitDegrees=45
  swing2Motion=Free
  twistLimitDegrees=45
  twistMotion=Free
 }
 {
  bone1Name=upperarm_l
  bone2Name=clavicle_l
  swing1LimitDegrees=45
  swing1Motion=Free
  swing2LimitDegrees=45
  swing2Motion=Free
  twistLimitDegrees=45
  twistMotion=Free
 }
 {
  bone1Name=thigh_r
  bone2Name=pelvis
  swing1LimitDegrees=45
  swing1Motion=Free
  swing2LimitDegrees=45
  swing2Motion=Free
  twistLimitDegrees=45
  twistMotion=Free
 }
 {bone1Name=clavicle_l, bone2Name=spine_05, swing1LimitDegrees=45, swing2LimitDegrees=30, twistLimitDegrees=15}
 {
  bone1Name=clavicle_r
  bone2Name=spine_05
  swing1LimitDegrees=45
  swing1Motion=Free
  swing2LimitDegrees=45
  swing2Motion=Free
  twistLimitDegrees=45
  twistMotion=Free
 }
 {
  bone1Name=upperarm_r
  bone2Name=clavicle_r
  swing1LimitDegrees=45
  swing1Motion=Free
  swing2LimitDegrees=45
  swing2Motion=Free
  twistLimitDegrees=45
  twistMotion=Free
 }
 {
  bone1Name=lowerarm_r
  bone2Name=upperarm_r
  swing1LimitDegrees=45
  swing1Motion=Free
  swing2LimitDegrees=45
  swing2Motion=Free
  twistLimitDegrees=45
  twistMotion=Free
 }
 {
  bone1Name=cape_l_MSTR
  bone2Name=clavicle_l
  swing1LimitDegrees=45
  swing1Motion=Free
  swing2LimitDegrees=45
  swing2Motion=Free
  twistLimitDegrees=45
  twistMotion=Free
 }
 {
  bone1Name=cape_r_MSTR
  bone2Name=clavicle_r
  swing1LimitDegrees=45
  swing1Motion=Free
  swing2LimitDegrees=45
  swing2Motion=Free
  twistLimitDegrees=45
  twistMotion=Free
 }
 {
  bone1Name=cape_mid_l_MSTR
  bone2Name=spine_05
  swing1LimitDegrees=45
  swing1Motion=Free
  swing2LimitDegrees=45
  swing2Motion=Free
  twistLimitDegrees=45
  twistMotion=Free
 }
 {
  bone1Name=cape_mid_r_MSTR
  bone2Name=spine_05
  swing1LimitDegrees=45
  swing1Motion=Free
  swing2LimitDegrees=45
  swing2Motion=Free
  twistLimitDegrees=45
  twistMotion=Free
 }
 {
  bone1Name=cape_front_l_MSTR
  bone2Name=spine_04
  swing1LimitDegrees=45
  swing1Motion=Free
  swing2LimitDegrees=45
  swing2Motion=Free
  twistLimitDegrees=45
  twistMotion=Free
 }
 {bone1Name=cape_front_01_l, bone2Name=cape_front_l_MSTR, swing1LimitDegrees=15, swing2LimitDegrees=45, twistLimitDegrees=2}
 {bone1Name=cape_front_02_l, bone2Name=cape_front_01_l, swing1LimitDegrees=15, swing2LimitDegrees=45, twistLimitDegrees=2}
 {bone1Name=cape_front_03_l, bone2Name=cape_front_02_l, swing1LimitDegrees=15, swing2LimitDegrees=45, twistLimitDegrees=2}
 {
  bone1Name=cape_front_mid_MSTR
  bone2Name=spine_04
  swing1LimitDegrees=45
  swing1Motion=Free
  swing2LimitDegrees=45
  swing2Motion=Free
  twistLimitDegrees=45
  twistMotion=Free
 }
 {bone1Name=cape_front_mid_01, bone2Name=cape_front_mid_MSTR, swing1LimitDegrees=15, swing2LimitDegrees=45, twistLimitDegrees=2}
 {bone1Name=cape_front_mid_02, bone2Name=cape_front_mid_01, swing1LimitDegrees=15, swing2LimitDegrees=45, twistLimitDegrees=2}
 {
  bone1Name=cape_front_r_MSTR
  bone2Name=spine_04
  swing1LimitDegrees=45
  swing1Motion=Free
  swing2LimitDegrees=45
  swing2Motion=Free
  twistLimitDegrees=45
  twistMotion=Free
 }
 {bone1Name=cape_front_01_r, bone2Name=cape_front_r_MSTR, swing1LimitDegrees=15, swing2LimitDegrees=45, twistLimitDegrees=2}
 {
  bone1Name=cape_mid_04_l
  bone2Name=cape_mid_04_r
  swing1LimitDegrees=45
  swing1Motion=Free
  swing2LimitDegrees=45
  swing2Motion=Free
  twistLimitDegrees=45
  twistMotion=Free
 }
 {
  bone1Name=cape_04_l
  bone2Name=cape_mid_04_l
  swing1LimitDegrees=45
  swing1Motion=Free
  swing2LimitDegrees=45
  swing2Motion=Free
  twistLimitDegrees=45
  twistMotion=Free
 }
]
