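# Physics asset for the dynamic back skirt of the male medium Lone Wolf Cape cosmetic body, built entirely from capsules.
# Left, center and right bone chains each run from a small MSTR body, angularly free on the pelvis, through base and three segments with tight swing and twist limits.
# Angularly free constraints tie both side tips to the center tip, and pelvis, thigh-twist and calf capsules supply body and leg volumes.
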PhysicsAsset /BRCosmetics/Characters/Player/Male/Medium/Bodies/M_MED_Lone_Wolf_Cape/Meshes/M_MED_Lone_Wolf_Cape_dyn_skirt
bodies {
 calf_l [{
   shapeType=Capsule
   length=43.25
   radius=7.25
   shapeName=thigh_twist_01_l_capsule
   center=<-24.1531 2.7424 -9.7815>
   rotation=<rot 85.087 -0.0674 0.5007>
  }]
 calf_r [{
   shapeType=Capsule
   length=15
   radius=19.25
   shapeName=thigh_twist_01_l_capsule
   center=<30.4185 -0.0071 25.653>
   rotation=<rot 84.9829 -0.0774 -179.5089>
  }]
 dyn_skirt_bk_01 [{shapeType=Capsule, length=25.3339, radius=7.9215, shapeName=dyn_skirt_bk_l_01_capsule, center=<8 0 5>, rotation=<rot 0 0 90>}]
 dyn_skirt_bk_02 [{shapeType=Capsule, length=25.3339, radius=7.9215, shapeName=dyn_skirt_bk_l_02_capsule, center=<8 0 5>, rotation=<rot 0 0 90>}]
 dyn_skirt_bk_03 [{shapeType=Capsule, length=25.3339, radius=7.9215, shapeName=dyn_skirt_bk_l_03_capsule, center=<8 0 5>, rotation=<rot 0 0 90>}]
 dyn_skirt_bk_MSTR [{shapeType=Capsule, length=0.505, radius=0.505, shapeName=dyn_skirt_bk_r_MSTR_capsule, rotation=<rot 0 -90 0>}]
 dyn_skirt_bk_base [{shapeType=Capsule, length=25.3339, radius=7.9215, shapeName=dyn_skirt_bk_l_01_capsule, center=<8 0 5>, rotation=<rot 0 0 90>}]
 dyn_skirt_bk_l_01 [{shapeType=Capsule, length=12.75, radius=7.75, shapeName=dyn_skirt_bk_l_01_capsule, center=<8 0 7>, rotation=<rot 0 0 79.9988>}]
 dyn_skirt_bk_l_02 [{shapeType=Capsule, length=12.75, radius=7.75, shapeName=dyn_skirt_bk_l_02_capsule, center=<8 0 7>, rotation=<rot 0 0 79.9972>}]
 dyn_skirt_bk_l_03 [{shapeType=Capsule, length=12.75, radius=7.75, shapeName=dyn_skirt_bk_l_03_capsule, center=<8 0 7>, rotation=<rot 0 0 79.9972>}]
 dyn_skirt_bk_l_MSTR [{shapeType=Capsule, length=0.505, radius=0.505, shapeName=dyn_skirt_bk_l_MSTR_capsule, rotation=<rot 0 -90 0>}]
 dyn_skirt_bk_l_base [{shapeType=Capsule, length=12.75, radius=7.75, shapeName=dyn_skirt_bk_l_01_capsule, center=<8 0 7>, rotation=<rot 0 0 79.9988>}]
 dyn_skirt_bk_r_01 [{shapeType=Capsule, length=12, radius=7.75, shapeName=dyn_skirt_bk_l_01_capsule, center=<8 0 7>, rotation=<rot -0.0017 -0.0003 100.0038>}]
 dyn_skirt_bk_r_02 [{shapeType=Capsule, length=12, radius=7.75, shapeName=dyn_skirt_bk_l_02_capsule, center=<8 0 7>, rotation=<rot -0.0017 -0.0003 100.0042>}]
 dyn_skirt_bk_r_03 [{shapeType=Capsule, length=12, radius=7.75, shapeName=dyn_skirt_bk_l_03_capsule, center=<8 0 7>, rotation=<rot -0.0017 -0.0003 100.0047>}]
 dyn_skirt_bk_r_MSTR [{shapeType=Capsule, length=0.505, radius=0.505, shapeName=dyn_skirt_bk_r_MSTR_capsule, rotation=<rot 0 -90 0>}]
 dyn_skirt_bk_r_base [{shapeType=Capsule, length=12, radius=7.75, shapeName=dyn_skirt_bk_l_01_capsule, center=<8 0 7>, rotation=<rot -0.0017 -0.0003 100.0037>}]
 pelvis [{shapeType=Capsule, length=10.3112, radius=15.2732, shapeName=pelvis_capsule, center=<-7.3649 4.243 0>}]
 thigh_twist_01_l [{
   shapeType=Capsule
   length=21.682
   radius=12.4857
   shapeName=thigh_twist_01_l_capsule
   center=<2.985 2.323 1.5391>
   rotation=<rot 90 0 -1.9773>
  }]
 thigh_twist_01_r [{
   shapeType=Capsule
   length=21.682
   radius=12.4857
   shapeName=thigh_twist_01_l_capsule
   center=<-2.985 -2.323 -1.5391>
   rotation=<rot 90 0 178.0226>
  }]
}
constraints [
 {bone1Name=dyn_skirt_bk_l_01, bone2Name=dyn_skirt_bk_l_base, swing1LimitDegrees=5, swing2LimitDegrees=10, twistLimitDegrees=2}
 {bone1Name=dyn_skirt_bk_l_02, bone2Name=dyn_skirt_bk_l_01, swing1LimitDegrees=5, swing2LimitDegrees=10, twistLimitDegrees=2}
 {bone1Name=dyn_skirt_bk_l_03, bone2Name=dyn_skirt_bk_l_02, swing1LimitDegrees=5, swing2LimitDegrees=10, twistLimitDegrees=2}
 {bone1Name=dyn_skirt_bk_r_01, bone2Name=dyn_skirt_bk_r_base, swing1LimitDegrees=5, swing2LimitDegrees=10, twistLimitDegrees=5}
 {bone1Name=dyn_skirt_bk_r_02, bone2Name=dyn_skirt_bk_r_01, swing1LimitDegrees=5, swing2LimitDegrees=10, twistLimitDegrees=5}
 {bone1Name=dyn_skirt_bk_r_03, bone2Name=dyn_skirt_bk_r_02, swing1LimitDegrees=5, swing2LimitDegrees=10, twistLimitDegrees=5}
 {bone1Name=dyn_skirt_bk_02, bone2Name=dyn_skirt_bk_01, swing1LimitDegrees=5, swing2LimitDegrees=10, twistLimitDegrees=2}
 {bone1Name=dyn_skirt_bk_03, bone2Name=dyn_skirt_bk_02, swing1LimitDegrees=5, swing2LimitDegrees=10, twistLimitDegrees=2}
 {bone1Name=dyn_skirt_bk_01, bone2Name=dyn_skirt_bk_base, swing1LimitDegrees=5, swing2LimitDegrees=10, twistLimitDegrees=2}
 {
  bone1Name=dyn_skirt_bk_r_03
  bone2Name=dyn_skirt_bk_03
  swing1LimitDegrees=45
  swing1Motion=Free
  swing2LimitDegrees=45
  swing2Motion=Free
  twistLimitDegrees=45
  twistMotion=Free
 }
 {
  bone1Name=dyn_skirt_bk_l_03
  bone2Name=dyn_skirt_bk_03
  swing1LimitDegrees=45
  swing1Motion=Free
  swing2LimitDegrees=45
  swing2Motion=Free
  twistLimitDegrees=45
  twistMotion=Free
 }
 {
  bone1Name=dyn_skirt_bk_MSTR
  bone2Name=pelvis
  swing1LimitDegrees=45
  swing1Motion=Free
  swing2LimitDegrees=45
  swing2Motion=Free
  twistLimitDegrees=45
  twistMotion=Free
 }
 {
  bone1Name=dyn_skirt_bk_l_MSTR
  bone2Name=pelvis
  swing1LimitDegrees=45
  swing1Motion=Free
  swing2LimitDegrees=45
  swing2Motion=Free
  twistLimitDegrees=45
  twistMotion=Free
 }
 {
  bone1Name=dyn_skirt_bk_r_MSTR
  bone2Name=pelvis
  swing1LimitDegrees=45
  swing1Motion=Free
  swing2LimitDegrees=45
  swing2Motion=Free
  twistLimitDegrees=45
  twistMotion=Free
 }
 {bone1Name=dyn_skirt_bk_l_base, bone2Name=dyn_skirt_bk_l_MSTR, swing1LimitDegrees=5, swing2LimitDegrees=10, twistLimitDegrees=2}
 {bone1Name=dyn_skirt_bk_r_base, bone2Name=dyn_skirt_bk_r_MSTR, swing1LimitDegrees=5, swing2LimitDegrees=10, twistLimitDegrees=5}
 {bone1Name=dyn_skirt_bk_base, bone2Name=dyn_skirt_bk_MSTR, swing1LimitDegrees=5, swing2LimitDegrees=10, twistLimitDegrees=2}
 {
  bone1Name=calf_r
  bone2Name=pelvis
  swing1LimitDegrees=45
  swing1Motion=Free
  swing2LimitDegrees=45
  swing2Motion=Free
  twistLimitDegrees=45
  twistMotion=Free
 }
 {
  bone1Name=calf_l
  bone2Name=pelvis
  swing1LimitDegrees=45
  swing1Motion=Free
  swing2LimitDegrees=45
  swing2Motion=Free
  twistLimitDegrees=45
  twistMotion=Free
 }
 {
  bone1Name=thigh_twist_01_r
  bone2Name=pelvis
  swing1LimitDegrees=45
  swing1Motion=Free
  swing2LimitDegrees=45
  swing2Motion=Free
  twistLimitDegrees=45
  twistMotion=Free
 }
 {
  bone1Name=thigh_twist_01_l
  bone2Name=pelvis
  swing1LimitDegrees=45
  swing1Motion=Free
  swing2LimitDegrees=45
  swing2Motion=Free
  twistLimitDegrees=45
  twistMotion=Free
 }
]
