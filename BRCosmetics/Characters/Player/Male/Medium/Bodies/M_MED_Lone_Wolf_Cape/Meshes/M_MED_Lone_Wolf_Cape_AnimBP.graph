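# Animation Blueprint for the Lone Wolf cape body, adding secondary motion to cape, skirt, katanas, collar and arm pads.
# Copies the parent mesh pose through a deform Control Rig, AnimDynamics and corrective bone offsets weighted by limb and head rotations.
# Skirt then cape RigidBody passes switch physics assets for crouching and backpacks.
# With rigid bodies disabled, constraints, look-ats and backward-movement trails substitute, with a separate downed (DBNO) pose.
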AnimBlueprint /BRCosmetics/Characters/Player/Male/Medium/Bodies/M_MED_Lone_Wolf_Cape/Meshes/M_MED_Lone_Wolf_Cape_AnimBP
target_skeleton=&/M_MED_Lone_Wolf_Cape_Skeleton
graphs [
 AnimGraph {nodes [
   n1 = "Transform (Modify) Bone - Bone: dyn_skirt_bk_l_MSTR"(
    ComponentPose=Local To Component(LocalPose=Use cached pose 'prePose'())
    Alpha=Get thigh_rot_bk_l() -> (thigh_rot_bk_l:double)
    Translation="0,0,0"
    Rotation="30,-20,0"
    Scale="1,1,1"
   ) {
    rotationMode=BMM_Additive
    rotationSpace=BCS_BoneSpace
    translationMode=BMM_Additive
    translationSpace=BCS_BoneSpace
    alphaBoolBlend {blendOption=Linear}
    boneToModify {boneName=dyn_skirt_bk_l_MSTR}
    rotation {pitch=30, yaw=-20}
   }
   n2 = "Transform (Modify) Bone - Bone: dyn_skirt_bk_r_MSTR"(ComponentPose=@n1, Alpha=Get thigh_rot_bk_r() -> (thigh_rot_bk_r:double), Translation="0,0,0", Rotation="30,0,0", Scale="1,1,1") {
    rotationMode=BMM_Additive
    rotationSpace=BCS_BoneSpace
    translationMode=BMM_Additive
    translationSpace=BCS_BoneSpace
    alphaBoolBlend {blendOption=Linear}
    boneToModify {boneName=dyn_skirt_bk_r_MSTR}
    rotation {pitch=30}
   }
   n3 = "Transform (Modify) Bone - Bone: dyn_skirt_bk_l_MSTR"(
    ComponentPose=@n2
    Alpha=Get thigh_rot_out_l() -> (thigh_rot_out_l:double)
    Translation="0,-0.5,1"
    Rotation="11.161,-0.000,-5.471"
    Scale="1,1,1"
   ) {
    rotationMode=BMM_Additive
    rotationSpace=BCS_BoneSpace
    translationMode=BMM_Additive
    translationSpace=BCS_BoneSpace
    alphaBoolBlend {blendOption=Linear}
    boneToModify {boneName=dyn_skirt_bk_l_MSTR}
    rotation {pitch=11.161, roll=-5.471}
    translation {y=-0.5, z=1}
   }
   n4 = "Transform (Modify) Bone - Bone: dyn_skirt_bk_r_MSTR"(
    ComponentPose=@n3
    Alpha=Get thigh_rot_out_r() -> (thigh_rot_out_r:double)
    Translation="0,0.5,0.751"
    Rotation="11.161,-0.000,5.471"
    Scale="1,1,1"
   ) {
    rotationMode=BMM_Additive
    rotationSpace=BCS_BoneSpace
    translationMode=BMM_Additive
    translationSpace=BCS_BoneSpace
    alphaBoolBlend {blendOption=Linear}
    boneToModify {boneName=dyn_skirt_bk_r_MSTR}
    rotation {pitch=11.161, roll=5.471}
    translation {y=0.5, z=0.751}
   }
   n5 = "Constraint - dyn_skirt_bk_base (PP)"!(ComponentPose=@n4, Alpha=1, ConstraintWeights_0=0.5, ConstraintWeights_1:float=0.5) -> (Pose) {
    alphaBoolBlend {blendOption=Linear}
    boneToModify {boneName=dyn_skirt_bk_base}
    constraintSetup [{targetBone {boneName=dyn_skirt_bk_l_MSTR}}, {targetBone {boneName=dyn_skirt_bk_r_MSTR}}]
    constraintWeights [1,1]
   }
   n6 = "Constraint - cape_mid_02_l (PP)"!(ComponentPose=@n5, Alpha=1, ConstraintWeights_0=1, ConstraintWeights_1:float=0) -> (Pose) {
    alphaBoolBlend {blendOption=Linear}
    boneToModify {boneName=cape_mid_02_l}
    constraintSetup [{targetBone {boneName=spine_03}}, {targetBone {boneName=dyn_skirt_bk_r_MSTR}}]
    constraintWeights [1,1]
   }
   n7 = "Constraint - cape_mid_02_r (PP)"!(ComponentPose=@n6, Alpha=1, ConstraintWeights_0=1, ConstraintWeights_1:float=0) -> (Pose) {
    alphaBoolBlend {blendOption=Linear}
    boneToModify {boneName=cape_mid_02_r}
    constraintSetup [{targetBone {boneName=spine_03}}, {targetBone {boneName=dyn_skirt_bk_r_MSTR}}]
    constraintWeights [1,1]
   }
   n8 = "Constraint - cape_02_l (PP)"!(ComponentPose=@n7, Alpha=1, ConstraintWeights_0=0.3, ConstraintWeights_1:float=0.7) -> (Pose) {
    alphaBoolBlend {blendOption=Linear}
    boneToModify {boneName=cape_02_l}
    constraintSetup [{targetBone {boneName=upperarm_twist_01_l}}, {targetBone {boneName=spine_05}}]
    constraintWeights [1,1]
   }
   n9 = Look At - Bone: cape_02_r(ComponentPose=@n8, Alpha=1) {
    alphaBoolBlend {blendInTime=0.3, blendOption=Linear}
    boneToModify {boneName=cape_02_r}
    lookAtLocation=<0 -20 18>
    lookAtTarget {boneReference {boneName=pelvis}}
    lookAt_Axis {axis=<1 0>}
   }
   n10 = Look At - Bone: cape_02_l(ComponentPose=@n9, Alpha=0.5) {
    alphaBoolBlend {blendInTime=0.3, blendOption=Linear}
    boneToModify {boneName=cape_02_l}
    lookAtLocation=<120 -60 -50>
    lookAtTarget {boneReference {boneName=root}}
    lookAt_Axis {axis=<1 0>}
   }
   n11 = Look At - Bone: cape_mid_03_l(ComponentPose=@n10, Alpha=1) {
    alphaBoolBlend {blendInTime=0.3, blendOption=Linear}
    boneToModify {boneName=cape_mid_03_l}
    lookAtLocation=<-20 -30 -8>
    lookAtTarget {boneReference {boneName=pelvis}}
    lookAt_Axis {axis=<1 0>}
   }
   n12 = Look At - Bone: cape_mid_03_r(ComponentPose=@n11, Alpha=1) {
    alphaBoolBlend {blendInTime=0.3, blendOption=Linear}
    boneToModify {boneName=cape_mid_03_r}
    lookAtLocation=<-20 -30 8>
    lookAtTarget {boneReference {boneName=pelvis}}
    lookAt_Axis {axis=<1 0>}
   }
   n13 = Get bIsCrouching()
   n14 = Look At - Bone: dyn_skirt_bk_l_02!(ComponentPose=@n12, bAlphaBoolEnabled:bool=@n13) -> (Pose) {
    alphaInputType=Bool
    alphaBoolBlend {blendInTime=0.3, blendOption=Linear}
    boneToModify {boneName=dyn_skirt_bk_l_02}
    lookAtLocation=<35 -45 -50>
    lookAtTarget {boneReference {boneName=root}}
    lookAt_Axis {axis=<1 0>}
   }
   n15 = Look At - Bone: dyn_skirt_bk_02!(ComponentPose=@n14, bAlphaBoolEnabled:bool=@n13) -> (Pose) {
    alphaInputType=Bool
    alphaBoolBlend {blendInTime=0.3, blendOption=Linear}
    boneToModify {boneName=dyn_skirt_bk_02}
    lookAtLocation=<0 -70 -50>
    lookAtTarget {boneReference {boneName=root}}
    lookAt_Axis {axis=<1 0>}
   }
   n16 = Look At - Bone: dyn_skirt_bk_r_02!(ComponentPose=@n15, bAlphaBoolEnabled:bool=@n13) -> (Pose) {
    alphaInputType=Bool
    alphaBoolBlend {blendInTime=0.3, blendOption=Linear}
    boneToModify {boneName=dyn_skirt_bk_r_02}
    lookAtLocation=<-10 -70 -50>
    lookAtTarget {boneReference {boneName=root}}
    lookAt_Axis {axis=<1 0>}
   }
   n17 = Get backLoin_deform_alpha() -> (backLoin_deform_alpha:double)
   n18 = "Transform (Modify) Bone - Bone: dyn_skirt_bk_r_01"(ComponentPose=@n16, Alpha=@n17, Translation="0,2,0", Rotation="0,10,0", Scale="1,1,1") {
    rotationMode=BMM_Additive
    rotationSpace=BCS_BoneSpace
    scaleMode=BMM_Additive
    scaleSpace=BCS_BoneSpace
    translationMode=BMM_Additive
    translationSpace=BCS_BoneSpace
    alphaBoolBlend {blendOption=Linear}
    alphaScaleBiasClamp {bInterpResult=true, interpSpeedDecreasing=5}
    boneToModify {boneName=dyn_skirt_bk_r_01}
    rotation {yaw=10}
    translation {y=2}
   }
   n19 = "Transform (Modify) Bone - Bone: dyn_skirt_bk_r_02"(ComponentPose=@n18, Alpha=@n17, Translation="0,0,2", Rotation="-20,15,0", Scale="1,1,1") {
    rotationMode=BMM_Additive
    rotationSpace=BCS_BoneSpace
    scaleMode=BMM_Additive
    scaleSpace=BCS_BoneSpace
    translationMode=BMM_Additive
    translationSpace=BCS_BoneSpace
    alphaBoolBlend {blendOption=Linear}
    alphaScaleBiasClamp {bInterpResult=true, interpSpeedDecreasing=5}
    boneToModify {boneName=dyn_skirt_bk_r_02}
    rotation {pitch=-20, yaw=15}
    translation {z=2}
   }
   n20 = "Transform (Modify) Bone - Bone: dyn_skirt_bk_02"(ComponentPose=@n19, Alpha=@n17, Translation="0,0,0", Rotation="-15,30,0", Scale="1,1,1") {
    rotationMode=BMM_Additive
    rotationSpace=BCS_BoneSpace
    scaleMode=BMM_Additive
    scaleSpace=BCS_BoneSpace
    translationMode=BMM_Additive
    translationSpace=BCS_BoneSpace
    alphaBoolBlend {blendOption=Linear}
    alphaScaleBiasClamp {bInterpResult=true, interpSpeedDecreasing=5}
    boneToModify {boneName=dyn_skirt_bk_02}
    rotation {pitch=-15, yaw=30}
   }
   n21 = Get BwdMovement() -> (BwdMovement:double)
   n22 = Trail controller - Bone: dyn_skirt_bk_l_03(ComponentPose=@n20, Alpha=@n21) {
    chainLength=3
    maxDeltaTime=0.0333
    alphaScaleBiasClamp {bInterpResult=true, interpSpeedIncreasing=5}
    planarLimits [{drivingBone {boneName=pelvis}, planeTransform {location=<0 -11 0>, rotation=<rot 0 -5.0001 -89.9999>, scale=<0.5 0.5 0.5>}}]
    rotationLimits [{limitMax=<180 180 180>}, {limitMax=<180 180 180>}, {limitMax=<180 180 180>}]
    rotationOffsets [<0 0 0>, <0 0 0>, <0 0 0>]
    trailBone {boneName=dyn_skirt_bk_l_03}
    trailRelaxationSpeed {editorCurveData {keys [<at 0 20 RCIM_Linear>,<at 0.5055 25 RCIM_Linear>,<at 1 5 RCIM_Linear>]}}
   }
   n23 = Trail controller - Bone: dyn_skirt_bk_r_03(ComponentPose=@n22, Alpha=@n21) {
    chainLength=3
    maxDeltaTime=0.0333
    alphaScaleBiasClamp {bInterpResult=true, interpSpeedIncreasing=5}
    planarLimits [{drivingBone {boneName=pelvis}, planeTransform {location=<0 -11 0>, rotation=<rot 0 -5.0001 -89.9999>, scale=<0.5 0.5 0.5>}}]
    rotationLimits [{limitMax=<180 180 180>}, {limitMax=<180 180 180>}, {limitMax=<180 180 180>}]
    rotationOffsets [<0 0 0>, <0 0 0>, <0 0 0>]
    trailBone {boneName=dyn_skirt_bk_r_03}
    trailRelaxationSpeed {editorCurveData {keys [<at 0 25 RCIM_Linear>,<at 0.5055 25 RCIM_Linear>,<at 1 5 RCIM_Linear>]}}
   }
   n24 = Trail controller - Bone: dyn_skirt_bk_03(ComponentPose=@n23, Alpha=@n21) {
    chainLength=3
    maxDeltaTime=0.0333
    alphaScaleBiasClamp {bInterpResult=true, interpSpeedIncreasing=5}
    planarLimits [{drivingBone {boneName=pelvis}, planeTransform {location=<0 -11 0>, rotation=<rot 0 -5.0001 -89.9999>, scale=<0.5 0.5 0.5>}}]
    rotationLimits [{limitMax=<180 180 180>}, {limitMax=<180 180 180>}, {limitMax=<180 180 180>}]
    rotationOffsets [<0 0 0>, <0 0 0>, <0 0 0>]
    trailBone {boneName=dyn_skirt_bk_03}
    trailRelaxationSpeed {editorCurveData {keys [<at 0 25 RCIM_Linear>,<at 0.5055 25 RCIM_Linear>,<at 1 5 RCIM_Linear>]}}
   }
   n25 = Trail controller - Bone: cape_06_l(ComponentPose=@n24, Alpha=@n21) {
    chainLength=6
    maxDeltaTime=0.0333
    alphaScaleBiasClamp {bInterpResult=true, interpSpeedIncreasing=5}
    planarLimits [{drivingBone {boneName=pelvis}, planeTransform {location=<0 -11 0>, rotation=<rot 0 -5.0001 -89.9999>, scale=<0.5 0.5 0.5>}}]
    rotationLimits [
     {limitMax=<180 180 180>}
     {limitMax=<180 180 180>}
     {limitMax=<180 180 180>}
     {limitMax=<180 180 180>}
     {limitMax=<180 180 180>}
     {limitMax=<180 180 180>}
    ]
    rotationOffsets [<0 0 0>, <0 0 0>, <0 0 0>, <0 0 0>, <0 0 0>, <0 0 0>]
    trailBone {boneName=cape_06_l}
    trailRelaxationSpeed {editorCurveData {keys [<at 0 25 RCIM_Linear>,<at 0.5055 25 RCIM_Linear>,<at 1 5 RCIM_Linear>]}}
   }
   n26 = Component To Local(ComponentPose=@n25)
   n27 = "Transform (Modify) Bone - Bone: dyn_skirt_bk_l_01"!(
    ComponentPose=Local To Component(LocalPose=Use cached pose 'prePose'())
    bAlphaBoolEnabled:bool=True
    Translation="0,0,0"
    Rotation="-25,25,0"
    Scale="1,1,1"
   ) -> (Pose) {
    alphaInputType=Bool
    rotationMode=BMM_Additive
    rotationSpace=BCS_BoneSpace
    translationMode=BMM_Additive
    translationSpace=BCS_BoneSpace
    alphaBoolBlend {blendOption=Linear}
    boneToModify {boneName=dyn_skirt_bk_l_01}
    rotation {pitch=-25, yaw=25}
   }
   n28 = "Transform (Modify) Bone - Bone: dyn_skirt_bk_l_02"!(ComponentPose=@n27, bAlphaBoolEnabled:bool=True, Translation="0,0,2", Rotation="-15,15,0.000", Scale="1,1,1") -> (Pose) {
    alphaInputType=Bool
    rotationMode=BMM_Additive
    rotationSpace=BCS_BoneSpace
    scaleMode=BMM_Additive
    scaleSpace=BCS_BoneSpace
    translationMode=BMM_Additive
    translationSpace=BCS_BoneSpace
    alphaBoolBlend {blendOption=Linear}
    boneToModify {boneName=dyn_skirt_bk_l_02}
    rotation {pitch=-15, yaw=15}
    translation {z=2}
   }
   n29 = "Transform (Modify) Bone - Bone: dyn_skirt_bk_r_01"!(ComponentPose=@n28, bAlphaBoolEnabled:bool=True, Translation="0,0,0", Rotation="-25,-25,0", Scale="1,1,1") -> (Pose) {
    alphaInputType=Bool
    rotationMode=BMM_Additive
    rotationSpace=BCS_BoneSpace
    translationMode=BMM_Additive
    translationSpace=BCS_BoneSpace
    alphaBoolBlend {blendOption=Linear}
    boneToModify {boneName=dyn_skirt_bk_r_01}
    rotation {pitch=-25, yaw=-25}
   }
   n30 = "Transform (Modify) Bone - Bone: dyn_skirt_bk_r_02"!(ComponentPose=@n29, bAlphaBoolEnabled:bool=True, Translation="0,0,2", Rotation="-15,-15,0.000", Scale="1,1,1") -> (Pose) {
    alphaInputType=Bool
    rotationMode=BMM_Additive
    rotationSpace=BCS_BoneSpace
    scaleMode=BMM_Additive
    scaleSpace=BCS_BoneSpace
    translationMode=BMM_Additive
    translationSpace=BCS_BoneSpace
    alphaBoolBlend {blendOption=Linear}
    boneToModify {boneName=dyn_skirt_bk_r_02}
    rotation {pitch=-15, yaw=-15}
    translation {z=2}
   }
   n31 = "Transform (Modify) Bone - Bone: dyn_skirt_bk_01"!(ComponentPose=@n30, bAlphaBoolEnabled:bool=True, Translation="0,0,0", Rotation="-25,0,0", Scale="1,1,1") -> (Pose) {
    alphaInputType=Bool
    rotationMode=BMM_Additive
    rotationSpace=BCS_BoneSpace
    translationMode=BMM_Additive
    translationSpace=BCS_BoneSpace
    alphaBoolBlend {blendOption=Linear}
    boneToModify {boneName=dyn_skirt_bk_01}
    rotation {pitch=-25}
   }
   n32 = "Transform (Modify) Bone - Bone: dyn_skirt_bk_02"!(ComponentPose=@n31, bAlphaBoolEnabled:bool=True, Translation="0,0,0", Rotation="-30,0,0.000", Scale="1,1,1") -> (Pose) {
    alphaInputType=Bool
    rotationMode=BMM_Additive
    rotationSpace=BCS_BoneSpace
    scaleMode=BMM_Additive
    scaleSpace=BCS_BoneSpace
    translationMode=BMM_Additive
    translationSpace=BCS_BoneSpace
    alphaBoolBlend {blendOption=Linear}
    boneToModify {boneName=dyn_skirt_bk_02}
    rotation {pitch=-30}
   }
   n33 = Look At - Bone: dyn_skirt_bk_l_01(ComponentPose=@n32, Alpha=1) {
    alphaBoolBlend {blendOption=Linear}
    boneToModify {boneName=dyn_skirt_bk_l_01}
    lookAtLocation {x=0}
    lookAtTarget {boneReference {boneName=foot_l}}
    lookAt_Axis {axis=<1 0>}
   }
   n34 = Look At - Bone: dyn_skirt_bk_r_01(ComponentPose=@n33, Alpha=1) {
    alphaBoolBlend {blendOption=Linear}
    boneToModify {boneName=dyn_skirt_bk_r_01}
    lookAtLocation {x=0}
    lookAtTarget {boneReference {boneName=foot_r}}
    lookAt_Axis {axis=<1 0>}
   }
   n35 = "Constraint - dyn_skirt_bk_01 (PPP)"!(ComponentPose=@n34, Alpha=1, ConstraintWeights_0=0.4, ConstraintWeights_1:float=0.4, ConstraintWeights_2:float=0.2) -> (Pose) {
    alphaBoolBlend {blendOption=Linear}
    boneToModify {boneName=dyn_skirt_bk_01}
    constraintSetup [{targetBone {boneName=dyn_skirt_bk_l_01}}, {targetBone {boneName=dyn_skirt_bk_r_01}}, {targetBone {boneName=dyn_skirt_bk_base}}]
    constraintWeights [1,1,1]
   }
   n36 = Component To Local(ComponentPose=@n35)
   n37 = Get bIsDBNO()
   n38 = Get GravityOverride()
   n39 = RigidBody!(ComponentPose=Local To Component(LocalPose=Use cached pose 'perRigidPose'()), Alpha, OverrideWorldGravity:Vector=@n38, ExternalForce) -> (Pose) {
    bClampLinearTranslationLimitToRefPose=true
    bOverrideWorldGravity=true
    overridePhysicsAsset=&/M_MED_Lone_Wolf_Cape_DynCape
    simulationSpace=BaseBoneSpace
    alphaBoolBlend {blendOption=Linear}
    baseBoneRef {boneName=spine_05}
    componentAppliedLinearAccClamp=<3000 3000 3000>
    componentLinearAccScale {x=0.4}
    componentLinearVelScale {x=0.4}
    overrideWorldGravity {z=-980}
    simSpaceSettings {
     maxAngularAcceleration=200
     maxAngularVelocity=2
     maxLinearAcceleration=800
     maxLinearVelocity=200
     velocityScaleZ=0
     worldAlpha=1
     externalLinearDragV=<1 1 1>
    }
   }
   n40 = Component To Local(ComponentPose=@n39)
   n41 = RigidBody(ComponentPose=Local To Component(LocalPose=Use cached pose 'perRigidPose'()), OverrideWorldGravity=@n38) {
    bClampLinearTranslationLimitToRefPose=true
    bOverrideWorldGravity=true
    overridePhysicsAsset=&/M_MED_Lone_Wolf_Cape_DynCape_packing
    simulationSpace=BaseBoneSpace
    alphaBoolBlend {blendOption=Linear}
    baseBoneRef {boneName=spine_05}
    componentAppliedLinearAccClamp=<3000 3000 3000>
    componentLinearAccScale {x=0.4}
    componentLinearVelScale {x=0.4}
    overrideWorldGravity {z=-980}
    simSpaceSettings {
     maxAngularAcceleration=100
     maxAngularVelocity=2
     maxLinearAcceleration=500
     maxLinearVelocity=200
     velocityScaleZ=0
     worldAlpha=1
     externalLinearDragV=<1 1 1>
    }
   }
   n42 = Component To Local(ComponentPose=@n41)
   n43 = Get wearingBackpack() -> (wearingBackpack:bool)
   n44 = "Transform (Modify) Bone - Bone: cape_l_MSTR"!(
    ComponentPose=Local To Component(LocalPose=Blend Poses by bool(BlendPose_0=@n42, BlendPose_1=@n40, bActiveValue=@n43) {blendType=Linear})
    bAlphaBoolEnabled:bool=Get capeAdjustAphla() -> (capeAdjustAphla:bool)
    Translation="0,0,0"
    Rotation="15,0.000,0.000"
    Scale="1,1,1"
   ) -> (Pose) {
    alphaInputType=Bool
    rotationMode=BMM_Additive
    rotationSpace=BCS_BoneSpace
    translationMode=BMM_Additive
    translationSpace=BCS_BoneSpace
    alphaBoolBlend {blendInTime=0.3, blendOption=Linear, blendOutTime=0.5}
    boneToModify {boneName=cape_l_MSTR}
    rotation {pitch=15}
   }
   n45 = Get bIsDBNO()
   n46 = "Transform (Modify) Bone - Bone: cape_mid_l_MSTR"!(ComponentPose=@n44, bAlphaBoolEnabled:bool=@n45, Translation="0,0,0", Rotation="3,0.000,0.000", Scale="1,1,1") -> (Pose) {
    alphaInputType=Bool
    rotationMode=BMM_Additive
    rotationSpace=BCS_BoneSpace
    translationMode=BMM_Additive
    translationSpace=BCS_BoneSpace
    alphaBoolBlend {blendInTime=0.3, blendOption=Linear, blendOutTime=0.5}
    boneToModify {boneName=cape_mid_l_MSTR}
    rotation {pitch=3}
   }
   n47 = "Transform (Modify) Bone - Bone: cape_mid_r_MSTR"!(ComponentPose=@n46, bAlphaBoolEnabled:bool=@n45, Translation="0,0,0", Rotation="3,0.000,0.000", Scale="1,1,1") -> (Pose) {
    alphaInputType=Bool
    rotationMode=BMM_Additive
    rotationSpace=BCS_BoneSpace
    translationMode=BMM_Additive
    translationSpace=BCS_BoneSpace
    alphaBoolBlend {blendInTime=0.3, blendOption=Linear, blendOutTime=0.5}
    boneToModify {boneName=cape_mid_r_MSTR}
    rotation {pitch=3}
   }
   n48 = "Transform (Modify) Bone - Bone: cape_r_MSTR"!(ComponentPose=@n47, bAlphaBoolEnabled:bool=@n45, Translation="0,0,0", Rotation="10,0.000,0.000", Scale="1,1,1") -> (Pose) {
    alphaInputType=Bool
    rotationMode=BMM_Additive
    rotationSpace=BCS_BoneSpace
    translationMode=BMM_Additive
    translationSpace=BCS_BoneSpace
    alphaBoolBlend {blendInTime=0.3, blendOption=Linear, blendOutTime=0.5}
    boneToModify {boneName=cape_r_MSTR}
    rotation {pitch=10}
   }
   Output Pose(Result=Blend Poses by bool(
     BlendPose_0=Component To Local(ComponentPose=@n48)
     BlendPose_1=Blend Poses by bool(BlendPose_0=@n36, BlendPose_1=@n26, bActiveValue=@n37) {blendType=Linear}
     bActiveValue=Get bRigidBodyEnabled()
    ) {blendType=Linear}) {layerGroup=None}
   n49 = Control Rig!(
    Source=Copy Pose From Mesh() {bUseAttachedParent=true}
    Alpha
    KneeRot_fr_l_114deg:Vector
    KneeTrans_fr_l_114deg:Vector="0.258,-3.591,0.009"
    KneeScale_fr_l_114deg:Vector="1,1,1"
    KneeScale_fr_l_130deg:Vector="1,1,1"
    KneeRot_fr_l_130deg:Vector="0,0,-51.2"
    KneeTrans_fr_l_130deg:Vector="0.258,-3.591,0.009"
   ) -> (Pose) {
    alphaBoolBlend {blendOption=Linear}
    controlRigAssetReference {blueprintRigClass=/BRCosmetics/Characters/Player/Male/Medium/Bodies/M_MED_Lone_Wolf/Meshes/M_MED_Lone_Wolf_Deform_CtrlRig_sculpters.M_MED_Lone_Wolf_Deform_CtrlRig_sculpters_C}
   }
   n50 = "Constraint - dyn_katanas_MSTR (PPP)"!(
    ComponentPose=Local To Component(LocalPose=@n49)
    Alpha=1
    ConstraintWeights_0=0.7218
    ConstraintWeights_1:float=0.1811
    ConstraintWeights_2:float=0.0971
   ) -> (Pose) {
    alphaBoolBlend {blendOption=Linear}
    boneToModify {boneName=dyn_katanas_MSTR}
    constraintSetup [{targetBone {boneName=pelvis}}, {targetBone {boneName=spine_01}}, {targetBone {boneName=dyn_skirt_ft_r_MSTR}}]
    constraintWeights [1,1,1]
   }
   n51 = Anim Dynamics - Bone: dyn_katanas_base(ComponentPose=@n50, Alpha=1, ExternalForce="0,0,0") {
    angularDampingOverride=0.999
    angularSpringConstant=2000
    bGravityOverrideInSimSpace=true
    bOverrideAngularDamping=true
    bUseGravityOverride=true
    gravityScale=0.35
    simulationSpace=RootRelative
    alphaBoolBlend {blendOption=Linear}
    boundBone {boneName=dyn_katanas_base}
    componentLinearAccScale=<0.2 0.2 0.2>
    componentLinearVelScale=<0.2 0.2 0.2>
    gravityOverride {z=-980}
    physicsBodyDefinitions [{
      sphereCollisionRadius=0
      boundBone {boneName=dyn_katanas_base}
      boxExtents=<10 15 20>
      constraintSetup {angularLimitsMax=<0 0 5>, angularLimitsMin=<0 0 -5>, angularTarget=<0 0 0>}
      localJointOffset=<1 0 0>
     }]
    relativeSpaceBone {boneName=dyn_katanas_MSTR}
   }
   n52 = Anim Dynamics - Bone: dyn_katanas_02(ComponentPose=@n51, Alpha=1, ExternalForce="0,0,0") {
    angularDampingOverride=0.999
    angularSpringConstant=1000
    bAngularSpring=true
    bGravityOverrideInSimSpace=true
    bOverrideAngularDamping=true
    bUseGravityOverride=true
    gravityScale=0.35
    simulationSpace=RootRelative
    alphaBoolBlend {blendOption=Linear}
    boundBone {boneName=dyn_katanas_02}
    componentLinearAccScale=<0.2 0.2 0.2>
    componentLinearVelScale=<0.2 0.2 0.2>
    gravityOverride {z=-980}
    physicsBodyDefinitions [{
      sphereCollisionRadius=0
      boundBone {boneName=dyn_katanas_02}
      boxExtents=<10 15 20>
      constraintSetup {angularLimitsMax=<0 2 5.03>, angularLimitsMin=<0 -2 -3>, angularTarget=<0 0 0>}
      localJointOffset=<2 0 0>
     }]
    relativeSpaceBone {boneName=dyn_katanas_MSTR}
   }
   n53 = Anim Dynamics - Bone: dyn_katanas_01(ComponentPose=@n52, Alpha=1, ExternalForce="0,0,0") {
    angularDampingOverride=0.999
    angularSpringConstant=1000
    bAngularSpring=true
    bGravityOverrideInSimSpace=true
    bOverrideAngularDamping=true
    bUseGravityOverride=true
    gravityScale=0.35
    simulationSpace=RootRelative
    alphaBoolBlend {blendOption=Linear}
    boundBone {boneName=dyn_katanas_01}
    componentLinearAccScale=<0.2 0.2 0.2>
    componentLinearVelScale=<0.2 0.2 0.2>
    gravityOverride {z=-980}
    physicsBodyDefinitions [{
      sphereCollisionRadius=0
      boundBone {boneName=dyn_katanas_01}
      boxExtents=<10 15 20>
      constraintSetup {angularLimitsMax=<0 2 5.03>, angularLimitsMin=<0 -2 -3>, angularTarget=<0 0 0>}
      localJointOffset=<1 0 0>
     }]
    relativeSpaceBone {boneName=dyn_katanas_MSTR}
   }
   n54 = Get thigh_rot_up_l() -> (thigh_rot_up_l:double)
   n55 = "Transform (Modify) Bone - Bone: dyn_skirt_ft_l_MSTR"(ComponentPose=@n53, Alpha=@n54, Translation="-6.5,0.000,-0.000", Rotation="0,0,0", Scale="1,1,1") {
    translationMode=BMM_Additive
    translationSpace=BCS_BoneSpace
    alphaBoolBlend {blendOption=Linear}
    boneToModify {boneName=dyn_skirt_ft_l_MSTR}
    translation {x=-6.5}
   }
   n56 = Get thigh_rot_up_r() -> (thigh_rot_up_r:double)
   n57 = "Transform (Modify) Bone - Bone: dyn_skirt_ft_r_MSTR"(ComponentPose=@n55, Alpha=@n56, Translation="-4,0.000,-0.000", Rotation="0,0,0", Scale="1,1,1") {
    translationMode=BMM_Additive
    translationSpace=BCS_BoneSpace
    alphaBoolBlend {blendOption=Linear}
    boneToModify {boneName=dyn_skirt_ft_r_MSTR}
    translation {x=-4}
   }
   n58 = Look At - Bone: dyn_skirt_ft_l_AIM(ComponentPose=@n57, Alpha=0.9) {
    alphaBoolBlend {blendOption=Linear}
    boneToModify {boneName=dyn_skirt_ft_l_AIM}
    lookAtLocation {x=0}
    lookAtTarget {boneReference {boneName=dyn_skirt_ft_l_TGT}}
    lookAt_Axis {axis=<1 0>}
   }
   n59 = Look At - Bone: dyn_skirt_ft_r_AIM(ComponentPose=@n58, Alpha=1) {
    alphaBoolBlend {blendOption=Linear}
    boneToModify {boneName=dyn_skirt_ft_r_AIM}
    lookAtLocation {x=0}
    lookAtTarget {boneReference {boneName=dyn_skirt_ft_r_TGT}}
    lookAt_Axis {axis=<1 0>}
   }
   n60 = "Transform (Modify) Bone - Bone: dyn_skirt_ft_l_AIM"(ComponentPose=@n59, Alpha=@n54, Translation="0,0,0", Rotation="0,25,-20", Scale="0.8,1.000,1") {
    rotationMode=BMM_Additive
    rotationSpace=BCS_BoneSpace
    scaleMode=BMM_Additive
    scaleSpace=BCS_BoneSpace
    translationMode=BMM_Additive
    translationSpace=BCS_BoneSpace
    alphaBoolBlend {blendOption=Linear}
    boneToModify {boneName=dyn_skirt_ft_l_AIM}
    rotation {roll=-20, yaw=25}
    scale {x=0.8}
   }
   n61 = "Transform (Modify) Bone - Bone: dyn_skirt_ft_r_AIM"(ComponentPose=@n60, Alpha=@n56, Translation="0,0,0", Rotation="0,0,0", Scale="0.8,1.000,1.000") {
    rotationMode=BMM_Additive
    rotationSpace=BCS_BoneSpace
    scaleMode=BMM_Additive
    scaleSpace=BCS_BoneSpace
    translationMode=BMM_Additive
    translationSpace=BCS_BoneSpace
    alphaBoolBlend {blendOption=Linear}
    boneToModify {boneName=dyn_skirt_ft_r_AIM}
    scale {x=0.8}
   }
   n62 = "Transform (Modify) Bone - Bone: dyn_skirt_ft_l_AIM"(ComponentPose=@n61, Alpha=Get thigh_rot_in_l() -> (thigh_rot_in_l:double), Translation="0,0,0", Rotation="0,-20,0", Scale="1,1,1") {
    rotationMode=BMM_Additive
    rotationSpace=BCS_BoneSpace
    scaleMode=BMM_Additive
    scaleSpace=BCS_BoneSpace
    translationMode=BMM_Additive
    translationSpace=BCS_BoneSpace
    alphaBoolBlend {blendOption=Linear}
    alphaScaleBiasClamp {bInterpResult=true}
    boneToModify {boneName=dyn_skirt_ft_l_AIM}
    rotation {yaw=-20}
   }
   n63 = "Transform (Modify) Bone - Bone: dyn_skirt_ft_r_AIM"(ComponentPose=@n62, Alpha=Get thigh_rot_in_r() -> (thigh_rot_in_r:double), Translation="0,0,0", Rotation="0,20,0", Scale="1,1,1") {
    rotationMode=BMM_Additive
    rotationSpace=BCS_BoneSpace
    scaleMode=BMM_Additive
    scaleSpace=BCS_BoneSpace
    translationMode=BMM_Additive
    translationSpace=BCS_BoneSpace
    alphaBoolBlend {blendOption=Linear}
    alphaScaleBiasClamp {bInterpResult=true}
    boneToModify {boneName=dyn_skirt_ft_r_AIM}
    rotation {yaw=20}
   }
   n64 = "Transform (Modify) Bone - Bone: dyn_skirt_ft_l_AIM"(ComponentPose=@n63, Alpha=Get thigh_rot_bk_l(), Translation="0,0,0", Rotation="25,10,0", Scale="1,1,1") {
    rotationMode=BMM_Additive
    rotationSpace=BCS_BoneSpace
    scaleMode=BMM_Additive
    scaleSpace=BCS_BoneSpace
    translationMode=BMM_Additive
    translationSpace=BCS_BoneSpace
    alphaBoolBlend {blendOption=Linear}
    boneToModify {boneName=dyn_skirt_ft_l_AIM}
    rotation {pitch=25, yaw=10}
   }
   n65 = "Transform (Modify) Bone - Bone: dyn_skirt_ft_r_AIM"(ComponentPose=@n64, Alpha=Get thigh_rot_bk_r(), Translation="0,0,0", Rotation="25,-10,0", Scale="1,1,1") {
    rotationMode=BMM_Additive
    rotationSpace=BCS_BoneSpace
    scaleMode=BMM_Additive
    scaleSpace=BCS_BoneSpace
    translationMode=BMM_Additive
    translationSpace=BCS_BoneSpace
    alphaBoolBlend {blendOption=Linear}
    boneToModify {boneName=dyn_skirt_ft_r_AIM}
    rotation {pitch=25, yaw=-10}
   }
   n66 = "Transform (Modify) Bone - Bone: deform_cape_front_r"(ComponentPose=@n65, Alpha=Get arm_rot_up_r() -> (arm_rot_up_r:double), Translation="-0.000,-15.000,6", Rotation="0,0,0", Scale="1,1,1") {
    rotationMode=BMM_Additive
    rotationSpace=BCS_BoneSpace
    scaleMode=BMM_Additive
    scaleSpace=BCS_BoneSpace
    translationMode=BMM_Additive
    translationSpace=BCS_BoneSpace
    alphaBoolBlend {blendOption=Linear}
    boneToModify {boneName=deform_cape_front_r}
    translation {y=-15, z=6}
   }
   n67 = "Transform (Modify) Bone - Bone: dyn_collar_ft"(ComponentPose=@n66, Alpha=Get head_rot_front() -> (head_rot_front:double), Translation="0,0,2", Rotation="25,0,0.000", Scale="0.6,1.5,1.000") {
    rotationMode=BMM_Additive
    rotationSpace=BCS_BoneSpace
    scaleMode=BMM_Additive
    scaleSpace=BCS_BoneSpace
    translationMode=BMM_Additive
    translationSpace=BCS_BoneSpace
    alphaBoolBlend {blendOption=Linear}
    boneToModify {boneName=dyn_collar_ft}
    rotation {pitch=25}
    scale=<0.6 1.5>
    translation {z=2}
   }
   n68 = "Transform (Modify) Bone - Bone: dyn_collar_ft"(ComponentPose=@n67, Alpha=1, Translation="0,0,0", Rotation=Get head_twist() -> (head_twist:Rotator), Scale="1,1,1") {
    rotationMode=BMM_Additive
    rotationSpace=BCS_BoneSpace
    scaleMode=BMM_Additive
    scaleSpace=BCS_BoneSpace
    translationMode=BMM_Additive
    translationSpace=BCS_BoneSpace
    alphaBoolBlend {blendOption=Linear}
    boneToModify {boneName=dyn_collar_ft}
    rotation {pitch=30, yaw=-20}
   }
   n69 = "Transform (Modify) Bone - Bone: deform_armPad_r"(ComponentPose=@n68, Alpha=Get arm_rot_up_r(), Translation="1.746,0,0.378", Rotation="-3.000,0.000,5.000", Scale="1,1,1") {
    rotationMode=BMM_Additive
    rotationSpace=BCS_BoneSpace
    scaleMode=BMM_Additive
    scaleSpace=BCS_BoneSpace
    translationMode=BMM_Additive
    translationSpace=BCS_BoneSpace
    alphaBoolBlend {blendOption=Linear}
    boneToModify {boneName=deform_armPad_r}
    rotation {pitch=-3, roll=5}
    translation {x=1.746, z=0.378}
   }
   n70 = "Transform (Modify) Bone - Bone: deform_armPad_l"(ComponentPose=@n69, Alpha=Get arm_rot_up_l() -> (arm_rot_up_l:double), Translation="3,0,0.3", Rotation="-2,20,-3", Scale="1,1,1") {
    rotationMode=BMM_Additive
    rotationSpace=BCS_BoneSpace
    scaleMode=BMM_Additive
    scaleSpace=BCS_BoneSpace
    translationMode=BMM_Additive
    translationSpace=BCS_BoneSpace
    alphaBoolBlend {blendOption=Linear}
    boneToModify {boneName=deform_armPad_l}
    rotation=<rot -2 20 -3>
    translation {x=3, z=0.3}
   }
   Save cached pose 'prePose'(Pose=Component To Local(ComponentPose=@n70)) {cacheName=prePose}
   n71 = "Transform (Modify) Bone - Bone: dyn_skirt_bk_r_MSTR"(
    ComponentPose=Local To Component(LocalPose=Use cached pose 'prePose'())
    Alpha=Get backLoin_deform_alpha()
    Translation="0,3,2"
    Rotation="0,0,0"
    Scale="1,1,1"
   ) {
    rotationMode=BMM_Additive
    rotationSpace=BCS_BoneSpace
    scaleMode=BMM_Additive
    scaleSpace=BCS_BoneSpace
    translationMode=BMM_Additive
    translationSpace=BCS_BoneSpace
    alphaBoolBlend {blendOption=Linear}
    boneToModify {boneName=dyn_skirt_bk_r_MSTR}
    translation {y=3, z=2}
   }
   n72 = Get GravityOverride()
   n73 = RigidBody(ComponentPose=@n71, OverrideWorldGravity=@n72) {
    bClampLinearTranslationLimitToRefPose=true
    bOverrideWorldGravity=true
    overridePhysicsAsset=&/M_MED_Lone_Wolf_Cape_dyn_skirt_crouching
    simulationSpace=BaseBoneSpace
    alphaBoolBlend {blendOption=Linear}
    baseBoneRef {boneName=spine_01}
    componentAppliedLinearAccClamp=<3000 3000 3000>
    componentLinearAccScale {x=0.45}
    componentLinearVelScale {x=0.45}
    overrideWorldGravity {z=-980}
    simSpaceSettings {
     maxAngularAcceleration=200
     maxAngularVelocity=5
     maxLinearAcceleration=1000
     maxLinearVelocity=200
     velocityScaleZ=0
     worldAlpha=1
     externalLinearDragV=<1 1>
    }
   }
   n74 = Component To Local(ComponentPose=@n73)
   n75 = RigidBody(ComponentPose=Local To Component(LocalPose=Use cached pose 'prePose'()), OverrideWorldGravity=@n72) {
    bClampLinearTranslationLimitToRefPose=true
    bOverrideWorldGravity=true
    overridePhysicsAsset=&/M_MED_Lone_Wolf_Cape_dyn_skirt
    simulationSpace=BaseBoneSpace
    alphaBoolBlend {blendOption=Linear}
    baseBoneRef {boneName=spine_01}
    componentAppliedLinearAccClamp=<3000 3000 3000>
    componentLinearAccScale {x=0.45}
    componentLinearVelScale {x=0.45}
    overrideWorldGravity {z=-980}
    simSpaceSettings {
     maxAngularAcceleration=200
     maxAngularVelocity=5
     maxLinearAcceleration=1000
     maxLinearVelocity=200
     velocityScaleZ=0
     worldAlpha=1
     externalLinearDragV=<1 1>
    }
   }
   n76 = Component To Local(ComponentPose=@n75)
   n77 = Get bIsCrouching()
   n78 = Get arm_rot_bk_l() -> (arm_rot_bk_l:double)
   n79 = "Transform (Modify) Bone - Bone: cape_l_MSTR"!(
    ComponentPose=Local To Component(LocalPose=Blend Poses by bool(BlendPose_0=@n74, BlendPose_1=@n76, bActiveValue=@n77) {blendType=Linear})
    Alpha:float=@n78
    Translation="0,-8,1.000"
    Rotation="0,0,0"
    Scale="1,1,1"
   ) -> (Pose) {
    rotationMode=BMM_Additive
    rotationSpace=BCS_BoneSpace
    scaleMode=BMM_Additive
    scaleSpace=BCS_BoneSpace
    translationMode=BMM_Additive
    translationSpace=BCS_BoneSpace
    alphaBoolBlend {blendOption=Linear}
    alphaScaleBiasClamp {bInterpResult=true}
    boneToModify {boneName=cape_l_MSTR}
    translation {y=-8, z=1}
   }
   n80 = "Transform (Modify) Bone - Bone: cape_mid_l_MSTR"!(ComponentPose=@n79, Alpha:float=@n78, Translation="0.000,-6,0", Rotation="0,0,0", Scale="1,1,1") -> (Pose) {
    rotationMode=BMM_Additive
    rotationSpace=BCS_BoneSpace
    scaleMode=BMM_Additive
    scaleSpace=BCS_BoneSpace
    translationMode=BMM_Additive
    translationSpace=BCS_BoneSpace
    alphaBoolBlend {blendOption=Linear}
    alphaScaleBiasClamp {bInterpResult=true}
    boneToModify {boneName=cape_mid_l_MSTR}
    translation {y=-6}
   }
   n81 = "Transform (Modify) Bone - Bone: cape_r_MSTR"!(
    ComponentPose=@n80
    Alpha:float=Get arm_rot_bk_r() -> (arm_rot_bk_r:double)
    Translation="0.000,6.936,4.232"
    Rotation="0,0,0"
    Scale="1,1,1"
   ) -> (Pose) {
    rotationMode=BMM_Additive
    rotationSpace=BCS_BoneSpace
    scaleMode=BMM_Additive
    scaleSpace=BCS_BoneSpace
    translationMode=BMM_Additive
    translationSpace=BCS_BoneSpace
    alphaBoolBlend {blendOption=Linear}
    alphaScaleBiasClamp {bInterpResult=true}
    boneToModify {boneName=cape_r_MSTR}
    translation {y=6.936, z=4.232}
   }
   n82 = Get floating_alpha() -> (floating_alpha:double)
   n83 = "Transform (Modify) Bone - Bone: cape_l_MSTR"(ComponentPose=@n81, Alpha=@n82, Translation="-1.984,0.157,3.787", Rotation="0,0,0", Scale="1,1,1") {
    rotationMode=BMM_Additive
    rotationSpace=BCS_BoneSpace
    scaleMode=BMM_Additive
    scaleSpace=BCS_BoneSpace
    translationMode=BMM_Additive
    translationSpace=BCS_BoneSpace
    alphaBoolBlend {blendOption=Linear}
    boneToModify {boneName=cape_l_MSTR}
    translation=<-1.984 0.1569 3.787>
   }
   n84 = "Transform (Modify) Bone - Bone: cape_r_MSTR"(ComponentPose=@n83, Alpha=@n82, Translation="-2.618,-1.667,1.910", Rotation="0,0,0", Scale="1,1,1") {
    rotationMode=BMM_Additive
    rotationSpace=BCS_BoneSpace
    scaleMode=BMM_Additive
    scaleSpace=BCS_BoneSpace
    translationMode=BMM_Additive
    translationSpace=BCS_BoneSpace
    alphaBoolBlend {blendOption=Linear}
    boneToModify {boneName=cape_r_MSTR}
    translation=<-2.618 -1.667 1.91>
   }
   Save cached pose 'perRigidPose'(Pose=Component To Local(ComponentPose=@n84)) {cacheName=perRigidPose}
  ]}
 EventGraph {nodes [
   Event BlueprintUpdateAnimation()
   > n1 = Sequence!(execute) -> (then_0, then_1, then_2:Exec, then_3:Exec, then_4:Exec, then_5:Exec, then_6:Exec, then_7:Exec, then_8:Exec)
   > n2 = Gravity Override State Function DEPRECATED(CharacterPartAnimInstance=Self-Reference(), GameTimeInSeconds=GetGameTimeInSeconds(), GravityOverrideParams=Get Gravity Override Params())
   > Set GravityOverride(GravityOverride=@n2)
   > n3 = Branch(Condition=Get bIsDBNO())
   > Set GravityOverride(GravityOverride="0,0,-680")
   Set GravityOverride(execute=@n3.else, GravityOverride=@n2)
   n4 = Branch(execute=@n1.then_1, Condition=Get bIsCrouching())
   n5 = GetOwnerSkeletalMeshComponent()
   n6 = BreakRotator(InRot=BreakTransform(InTransform=GetDeltaTransformFromRefPose(self=@n5, BoneName=thigh_r, BaseName=pelvis)).Rotation)
   > Set backLoin_deform_alpha(backLoin_deform_alpha:double=float * float(A=MapRangeClamped(Value=@n6.Yaw, InRangeA=0, InRangeB=-100, OutRangeA=1, OutRangeB=0), B=Get knee_rot_r() -> (knee_rot_r:double)) {is=title K2Node_PromotableOperator}) -> (Output_Get:double)
   Set backLoin_deform_alpha(execute=@n4.else, backLoin_deform_alpha=0)
   n7 = GetOwnerSkeletalMeshComponent()
   n8 = BreakRotator(InRot=BreakTransform(InTransform=GetDeltaTransformFromRefPose(self=@n7, BoneName=thigh_r, BaseName=pelvis)).Rotation)
   Set thigh_rot_up_r(execute=@n1.then_2, thigh_rot_up_r:double=MapRangeClamped(Value=@n8.Yaw, InRangeA=-15, InRangeB=-100)) -> (Output_Get:double)
   > Set thigh_rot_bk_r(thigh_rot_bk_r:double=MapRangeClamped(Value=@n8.Yaw, InRangeA=0, InRangeB=30)) -> (Output_Get:double)
   > Set thigh_rot_in_r(thigh_rot_in_r:double=MapRangeClamped(Value=@n8.Pitch, InRangeA=0, InRangeB=-30)) -> (Output_Get:double)
   > Set thigh_rot_out_r(thigh_rot_out_r:double=MapRangeClamped(Value=@n8.Pitch, InRangeA=0, InRangeB=20)) -> (Output_Get:double)
   n9 = GetOwnerSkeletalMeshComponent()
   n10 = BreakRotator(InRot=BreakTransform(InTransform=GetDeltaTransformFromRefPose(self=@n9, BoneName=thigh_l, BaseName=pelvis)).Rotation)
   > Set thigh_rot_up_l(thigh_rot_up_l:double=MapRangeClamped(Value=@n10.Yaw, InRangeA=-10, InRangeB=-100)) -> (Output_Get:double)
   > Set thigh_rot_bk_l(thigh_rot_bk_l:double=MapRangeClamped(Value=@n10.Yaw, InRangeA=0, InRangeB=30)) -> (Output_Get:double)
   > Set thigh_rot_in_l(thigh_rot_in_l:double=MapRangeClamped(Value=@n10.Pitch, InRangeA=0, InRangeB=-20)) -> (Output_Get:double)
   > Set thigh_rot_out_l(thigh_rot_out_l:double=MapRangeClamped(Value=@n10.Pitch, InRangeA=0, InRangeB=20)) -> (Output_Get:double)
   n11 = BreakTransform(InTransform=GetDeltaTransformFromRefPose(self=GetOwnerSkeletalMeshComponent(), BoneName=thigh_r, BaseName=calf_r))
   > Set knee_rot_r(knee_rot_r:double=MapRangeClamped(Value=BreakRotator(InRot=@n11.Rotation).Yaw, InRangeA=0, InRangeB=-90)) -> (Output_Get:double)
   n12 = GetOwnerSkeletalMeshComponent()
   n13 = BreakRotator(InRot=BreakTransform(InTransform=GetDeltaTransformFromRefPose(self=@n12, BoneName=head, BaseName=spine_05)).Rotation)
   > Set head_rot_front(head_rot_front:double=MapRangeClamped(Value=@n13.Yaw, InRangeA=0, InRangeB=20)) -> (Output_Get:double)
   > Set head_twist(head_twist:Rotator=MakeRotator(
     Roll=MapRangeClamped(Value=@n13, InRangeA=-25, InRangeB=25, OutRangeA=20, OutRangeB=-20)
     Pitch=MapRangeClamped(Value=@n13, InRangeA=-25, InRangeB=25, OutRangeA=5, OutRangeB=-5)
    )) -> (Output_Get:Rotator)
   n14 = GetOwnerSkeletalMeshComponent()
   n15 = BreakRotator(InRot=BreakTransform(InTransform=GetDeltaTransformFromRefPose(self=@n14, BoneName=upperarm_l, BaseName=spine_05)).Rotation)
   Set arm_rot_dn_l(execute=@n1.then_3, arm_rot_dn_l:double=MapRangeClamped(Value="Absolute (Float)"(A=@n15.Yaw), InRangeA=60, InRangeB=10)) -> (Output_Get:double)
   n16 = GetOwnerSkeletalMeshComponent()
   n17 = BreakRotator(InRot=BreakTransform(InTransform=GetDeltaTransformFromRefPose(self=@n16, BoneName=clavicle_l, BaseName=spine_05)).Rotation)
   > Set clavicle_rot_l(clavicle_rot_l:double=MapRangeClamped(Value="Absolute (Float)"(A=@n17.Yaw), InRangeA=0, InRangeB=20)) -> (Output_Get:double)
   n18 = GetOwnerSkeletalMeshComponent()
   n19 = BreakRotator(InRot=BreakTransform(InTransform=GetDeltaTransformFromRefPose(self=@n18, BoneName=upperarm_l, BaseName=spine_04)).Rotation)
   > Set arm_rot_bk_l(arm_rot_bk_l:double=MapRangeClamped(Value=@n19.Yaw, InRangeA=5, InRangeB=-15)) -> (Output_Get:double)
   n20 = GetOwnerSkeletalMeshComponent()
   n21 = BreakRotator(InRot=BreakTransform(InTransform=GetDeltaTransformFromRefPose(self=@n20, BoneName=upperarm_l, BaseName=spine_05)).Rotation)
   > Set arm_rot_up_l(arm_rot_up_l:double=MapRangeClamped(Value="Absolute (Float)"(A=@n21.Yaw), InRangeA=0, InRangeB=125)) -> (Output_Get:double)
   > Set cape_adjust_l(cape_adjust_l:double="Max (Float)"(A=Get clavicle_rot_l() -> (clavicle_rot_l:double), B=Get arm_rot_dn_l() -> (arm_rot_dn_l:double))) -> (Output_Get:double)
   n22 = GetOwnerSkeletalMeshComponent()
   n23 = BreakRotator(InRot=BreakTransform(InTransform=GetDeltaTransformFromRefPose(self=@n22, BoneName=upperarm_r, BaseName=spine_05)).Rotation)
   > Set arm_rot_up_r(arm_rot_up_r:double=MapRangeClamped(Value="Absolute (Float)"(A=@n23.Yaw), InRangeA=0, InRangeB=100)) -> (Output_Get:double)
   n24 = GetOwnerSkeletalMeshComponent()
   n25 = BreakRotator(InRot=BreakTransform(InTransform=GetDeltaTransformFromRefPose(self=@n24, BoneName=upperarm_r, BaseName=spine_04)).Rotation)
   > Set arm_rot_bk_r(arm_rot_bk_r:double=MapRangeClamped(Value="Absolute (Float)"(A=@n25.Yaw), InRangeA=0, InRangeB=30)) -> (Output_Get:double)
   n26 = Branch(execute=@n1.then_4, Condition=Get bIsSkydiveFloating())
   > Set floating_alpha(floating_alpha:double=1) -> (Output_Get:double)
   > Set arm_rot_bk_l(arm_rot_bk_l=0)
   > Set arm_rot_bk_r(arm_rot_bk_r=0)
   Set floating_alpha(execute=@n26.else, floating_alpha=0)
   n27 = Branch(execute=@n1.then_5, Condition=Get bIsInFrontEnd())
   > Set BwdMovement(BwdMovement:double=0.1) -> (Output_Get:double)
   n28 = Branch(execute=@n27.else, Condition=OR Boolean(A=Get bIsCrouching(), B=Get bIsSkyDiving()))
   > Set BwdMovement(BwdMovement=0)
   Set BwdMovement(execute=@n28.else, BwdMovement=Get BwdMovement())
   n29 = PlayerMovement(execute=@n1.then_6, AnimInstance=Self-Reference(), XDotProductCoefficient=1, YDotProductCoefficient=1, ZDotProductCoefficient=1)
   n30 = MapRangeClamped(Value=@n29.YMovement, InRangeA=0, InRangeB=1)
   n31 = float * float(A="Absolute (Float)"(A=@n29), B=0.5) {is=title K2Node_CommutativeAssociativeBinaryOperator}
   > Set BwdMovement(BwdMovement=MapRangeClamped(Value=float + float(A=@n31, B=@n30) {is=title K2Node_CommutativeAssociativeBinaryOperator}, InRangeA=0, InRangeB=1.1, OutRangeB=0.4))
   n32 = Cast To FortPlayerAnimInstance(execute=@n1.then_7, Object=GetAnimInstance(self=GetOwnerSkeletalMeshComponent()))
   > n33 = Switch on EFortWeaponCoreAnimation!(execute, Selection:EFortWeaponCoreAnimation Enum=Get WeaponCoreAnim(self=@n32)) -> (
    Melee:Exec
    Pistol:Exec
    Shotgun:Exec
    PaperBlueprint:Exec
    Rifle:Exec
    MeleeOneHand:Exec
    MachinePistol:Exec
    RocketLauncher:Exec
    GrenadeLauncher:Exec
    GoingCommando:Exec
    AssaultRifle:Exec
    TacticalShotgun:Exec
    SniperRifle:Exec
    TrapPlacement:Exec
    ShoulderLauncher:Exec
    AbilityDecoTool:Exec
    Crossbow:Exec
    C4:Exec
    RemoteControl:Exec
    DualWield:Exec
    AR_BullPup:Exec
    AR_ForwardGrip:Exec
    MedPackPaddles:Exec
    SMG_P90:Exec
    AR_DrumGun:Exec
    Consumable_Small:Exec
    Consumable_Large:Exec
    Balloon:Exec
    MountedTurret:Exec
    CreativeTool:Exec
    ExplosiveBow:Exec
    AshtonIndigo:Exec
    AshtonChicago:Exec
    MeleeDualWield:Exec
    Unarmed:Exec
   )
   Set OneHandPickaxe(
    execute=[
     @n33.AR_BullPup
     @n33.AR_DrumGun
     @n33.AR_ForwardGrip
     @n33.AbilityDecoTool
     @n33.AshtonChicago
     @n33.AshtonIndigo
     @n33.AssaultRifle
     @n33.Balloon
     @n33.C4
     @n33.Consumable_Large
     @n33.Consumable_Small
     @n33.CreativeTool
     @n33.Crossbow
     @n33.DualWield
     @n33.ExplosiveBow
     @n33.GoingCommando
     @n33.GrenadeLauncher
     @n33.MachinePistol
     @n33.MedPackPaddles
     @n33.Melee
     @n33.MeleeDualWield
     @n33.MountedTurret
     @n33.PaperBlueprint
     @n33.Pistol
     @n33.RemoteControl
     @n33.Rifle
     @n33.RocketLauncher
     @n33.SMG_P90
     @n33.Shotgun
     @n33.ShoulderLauncher
     @n33.SniperRifle
     @n33.TacticalShotgun
     @n33.TrapPlacement
    ]
    OneHandPickaxe:bool=false
   ) -> (Output_Get:bool)
   Set OneHandPickaxe(execute=@n33.MeleeOneHand, OneHandPickaxe=true)
   n34 = Branch(execute=@n1.then_8, Condition=Get bIsSprinting())
   > Set capeAdjustAphla(capeAdjustAphla:bool=true) -> (Output_Get:bool)
   Set capeAdjustAphla(execute=@n34.else, capeAdjustAphla=false)
   Event OnCharacterCustomizationLoaded()
   > Do Once()
   n35 = CheckForBackbling()
   > Set wearingBackpack(wearingBackpack:bool=OR Boolean(A=@n35, B=@n35.bIsBackblingACape)) -> (Output_Get:bool)
   Event OnWeaponChanged()
   > n36 = Get Dynamic Material Instance(SkeletalMeshComponent=GetPartSkeletalMeshComponent())
   n37 = PickaxeItemDefinitionHasAnyMetaTag(Tags=Make Array("[0]"=(TagName="Cosmetics.MetaTags.LoneWolf")), bRequiresEquipped=true)
   > n38 = Branch(Condition=@n37)
   > n39 = Branch(Condition=Get bIsSkyDiving())
   SetScalarParameterValue(execute=[@n40.then, @n39.then, @n41.else, @n38.else], self=@n36, ParameterName=Hide Element 01, Value=0)
   SetScalarParameterValue(execute=[@n39.else, @n41.then], self=@n36, ParameterName=Hide Element 01, Value=1)
   n40 = Event OnBeginSkydiving()
   Event OnEndSkydiving()
   > n41 = Branch(Condition=@n37)
   "Absolute (Float)"(A=@n19.Yaw)
  ]}
]
variables [
 arm_rot_bk_l {type=number}
 arm_rot_bk_r {type=number}
 arm_rot_dn_l {type=number}
 arm_rot_up_l {type=number}
 arm_rot_up_r {type=number}
 backLoin_deform_alpha {type=number}
 BwdMovement {type=number}
 cape_adjust_l {type=number}
 capeAdjustAphla {type=boolean}
 clavicle_rot_l {type=number}
 floating_alpha {type=number}
 Gravity Override Params {
  type=GravityOverrideParamsStruct
  default {skydive_FloatingDown {gravityOverride=<0 -700 100>, windAmplitude=<200 -500 200>, windOffset=<0 -600 200>}}
 }
 GravityOverride {type=Vector}
 head_rot_front {type=number}
 head_twist {type=Rotator}
 knee_rot_r {type=number}
 OneHandPickaxe {type=boolean}
 thigh_rot_bk_l {type=number}
 thigh_rot_bk_r {type=number}
 thigh_rot_in_l {type=number}
 thigh_rot_in_r {type=number}
 thigh_rot_out_l {type=number}
 thigh_rot_out_r {type=number}
 thigh_rot_up_l {type=number}
 thigh_rot_up_r {type=number}
 wearingBackpack {type=boolean}
]
