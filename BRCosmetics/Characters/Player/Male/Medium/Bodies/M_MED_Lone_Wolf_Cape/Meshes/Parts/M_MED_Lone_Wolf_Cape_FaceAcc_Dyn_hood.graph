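# Physics asset for the dynamic hood on the face accessory part of the male medium Lone Wolf Cape player cosmetic.
# Left and right hood bones each carry an offset, rotated capsule body, alongside a small capsule on the head bone.
# Two constraints tie each hood bone directly to the head, with moderate swing limits and a tight twist limit, so the two sides have no constraint between them.
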PhysicsAsset /BRCosmetics/Characters/Player/Male/Medium/Bodies/M_MED_Lone_Wolf_Cape/Meshes/Parts/M_MED_Lone_Wolf_Cape_FaceAcc_Dyn_hood
bodies {
 dyn_hood_l [{shapeType=Capsule, length=2, radius=2, center=<3 0 0>, rotation=<rot 90 0 0>}]
 dyn_hood_r [{shapeType=Capsule, length=2, radius=2, center=<3 0 0>, rotation=<rot 90 0 0>}]
 head [{shapeType=Capsule, length=1, radius=1, shapeName=head_capsule}]
}
constraints [
 {bone1Name=dyn_hood_r, bone2Name=head, swing1LimitDegrees=45, swing2LimitDegrees=45, twistLimitDegrees=5}
 {bone1Name=dyn_hood_l, bone2Name=head, swing1LimitDegrees=45, swing2LimitDegrees=45, twistLimitDegrees=5}
]
